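; Photo mode targeting filter that removes targets outside a view cone around the targeting pawn's aim, with an optional line-of-sight requirement.
; ShouldRemoveTarget takes the angle between aim forward and the direction from pawn location to impact point, removing targets at or beyond FieldOfView, which thus acts as a half-angle.
; When RequireLineOfSight is set, it also removes targets failing a Visibility-channel check from the pawn's eyes.
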(fn ShouldRemoveTarget
  (params (TargetingHandle :type Targeting%20Request%20Handle%20Structure%20%28by%20ref%29) (TargetData :type Targeting%20Default%20Result%20Data%20Structure%20%28by%20ref%29))
  (bind _hitresult (Utilities|Struct|BreakTargetingDefaultResultData TargetData))
  (bind (_hitresult_bblockinghit _hitresult_binitialoverlap _hitresult_time _hitresult_distance _hitresult_location _hitresult_impactpoint _hitresult_normal _hitresult_impactnormal _hitresult_physmat _hitresult_hitactor) (Collision|BreakHitResult _hitresult))
  (Utilities|IsValid _hitresult_hitactor
    (:Is%20Valid
      (LocalVariables|SetHitActor _hitresult_hitactor)
      (bind _asfort_pawn (Utilities|Casting|CastToFortPawn (Utilities|Struct|BreakTargetingSourceContext (TargetingSystem|TargetingTypes|GetTargetingSourceContext (Targeting|GetTargetingSubsystem TargetingHandle) TargetingHandle))))
      (bind _output_get (LocalVariables|SetTargetingPawn _asfort_pawn))
      (LocalVariables|SetRemoveFromFilter (>= (Math|Float|Absolute(Float) (Math|Trig|Acos(Degrees) (Math|Vector|DotProduct (Math|Vector|GetForwardVector (Pawn|GetBaseAimRotation _output_get)) (Math|Vector|Normalize (- _hitresult_impactpoint (Transformation|GetActorLocation (LocalVariables|GetTargetingPawn))))))) (Variables|Default|GetFieldOfView)))
      (if (Variables|Default|GetRequireLineOfSight)
        (bind _returnvalue (Fortnite|CheckLineOfSightToActorWithChannel (Actor|GetActorEyesViewPoint (LocalVariables|GetTargetingPawn)) (LocalVariables|GetHitActor) "ECC_Visibility" (LocalVariables|GetTargetingPawn)))
        (return (or (not _returnvalue) (LocalVariables|GetRemoveFromFilter)))
        (else
          (return (LocalVariables|GetRemoveFromFilter)))))))
